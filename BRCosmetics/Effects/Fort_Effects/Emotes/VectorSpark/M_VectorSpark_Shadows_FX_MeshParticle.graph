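# Unlit, two-sided masked material for Niagara mesh particles in the VectorSpark emote's shadow effect.
# Color blends from black to a contrast-boosted, rotated VectorSpark texture via the EmissiveSwap dynamic parameter.
# Panning noise feeds a foam texture lookup, combining with a scrolling gradient into hard-edged erosion driven by particle alpha.
# Vertex red and an optional sphere mask around the offset particle position trim the result before temporal AA dithering.
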Material /BRCosmetics/Effects/Fort_Effects/Emotes/VectorSpark/M_VectorSpark_Shadows_FX_MeshParticle
graph {
 properties {
  bCastRayTracedShadows=false
  bUsedWithNiagaraMeshParticles=true
  blendMode=BLEND_Masked
  pixelDepthOffsetMode=PDOM_Legacy
  shadingModel=MSM_Unlit
  twoSided=true
  usedShadingModels=Unlit
 }
 nodes [
  n1 = TextureCoordinate()
  n2 = AppendVector(A=OneMinus(None=ComponentMask(None=@n1) {g=true}), B=ComponentMask(None=@n1) {r=true})
  n3 = AppendVector(
   A=ScalarParameter() {defaultValue=0.5, group="3) Diffuse", parameterName=UV Scale U}
   B=ScalarParameter() {defaultValue=0.5, group="3) Diffuse", parameterName=UV Scale V}
  )
  n4 = ScalarParameter() {defaultValue=-0.065, group="3) Diffuse", parameterName=UV Add U}
  n5 = ScalarParameter() {defaultValue=0.016, group="3) Diffuse", parameterName=UV Add V}
  n6 = Rotator(Coordinate=Add(A=Multiply(A=@n2, B=@n3), B=AppendVector(A=@n4, B=@n5)), Time=ScalarParameter() {group="3) Diffuse", parameterName=Rotator})
  LinearInterpolate(
   A=Constant() {r=0}
   B=MaterialFunctionCall!(
    In=TextureSample(UVs=@n6) {samplerSource=SSM_Clamp_WorldGroupSettings, texture=/BRCosmetics/Animation/Props/Gadgets_VectorSpark/Textures/T_VectorSpark_E}.RGB
    Contrast=ScalarParameter() {defaultValue=4, group="3) Diffuse", parameterName="Contrast (S)"}
   ) {materialFunction=/Engine/Functions/Engine_MaterialFunctions01/ImageAdjustment/CheapContrast_RGB}.Result
   Alpha=LinearInterpolate(
    A=DynamicParameter() {defaultValue=<rgba 0 1 1 1>, paramNames [EmissiveSwap,Param2,Param3,Param4]}.EmissiveSwap
    B=ScalarParameter() {defaultValue=0.9, group="5) Debug", parameterName=Alpha, sliderMax=0.9, sliderMin=-0.35}
    Alpha=ScalarParameter() {group="5) Debug", parameterName=Debug Alpha}
   )
  )
  n7 = TextureCoordinate()
  n8 = AppendVector(
   A=ScalarParameter() {defaultValue=1, group="2) Noise", parameterName=Noise UV Scale U}
   B=ScalarParameter() {defaultValue=0.6, group="2) Noise", parameterName=Noise UV Scale V}
  )
  n9 = ScalarParameter() {defaultValue=0.1, group="2) Noise", parameterName=Noise Speed X}
  n10 = ScalarParameter() {defaultValue=-0.5, group="2) Noise", parameterName=Noise Speed Y}
  n11 = TextureSample(UVs=Panner(Coordinate=Multiply(A=@n7, B=@n8), Speed=AppendVector(A=@n9, B=@n10))) {samplerType=SAMPLERTYPE_Normal, texture=/Game/Effects/Fort_Effects/Emotes/BirdNestBlue/T_LowResSimplePerlin}
  n12 = Power(Base=TextureSample(UVs=Multiply(A=@n11.R) {constB=0.075}) {samplerType=SAMPLERTYPE_LinearColor, texture=/Game/Athena/Apollo/Water/Textures/02_Rounded/03_WaterFoam_Veroni_base_H}.R) {constExponent=4}
  n13 = TextureCoordinate()
  n14 = AppendVector(
   A=ScalarParameter() {defaultValue=1, group="1) Erosion", parameterName=Erosion UV Scale U}
   B=ScalarParameter() {defaultValue=1, group="1) Erosion", parameterName=Erosion UV Scale V}
  )
  n15 = ScalarParameter() {defaultValue=0.1, group="1) Erosion", parameterName=Erosion Speed X}
  n16 = ScalarParameter() {group="1) Erosion", parameterName=Erosion Speed Y}
  n17 = TextureSampleParameter2D(UVs=Panner(Coordinate=Multiply(A=@n13, B=@n14), Speed=AppendVector(A=@n15, B=@n16))) {
   group="1) Erosion"
   parameterName=ErosionTextureMask
   texture=&/T_VectorSpark_EorsionGradient
  }
  n18 = ParticleColor()
  n19 = ScalarParameter() {defaultValue=0.9, group="5) Debug", parameterName=Alpha, sliderMax=0.9, sliderMin=-0.35}
  n20 = ScalarParameter() {group="5) Debug", parameterName=Debug Alpha}
  n21 = Subtract(A=Add(A=Add(A=@n12, B=@n17.R)), B=LinearInterpolate(Alpha=LinearInterpolate(A=@n18.A, B=@n19, Alpha=@n20)) {constA=1.2, constB=0})
  n22 = Multiply(A=LinearInterpolate(Alpha=VertexColor().R) {constA=1, constB=0}, B=Clamp(None=Floor(None=@n21))) {constA=1}
  n23 = ComponentMask(None=VectorParameter() {group="3) Sphere Mask", parameterName=SphereMask Offset, defaultValue=<rgba 0 0 -8 0>}.RGB) {b=true, g=true, r=true}
  n24 = SphereMask(
   A=WorldPosition().XYZ
   B=Add(A=ParticlePositionWS(), B=Transform(None=@n23) {transformSourceType=TRANSFORMSOURCE_Local})
   Radius=ScalarParameter() {defaultValue=32, group="3) Sphere Mask", parameterName=Radius}
   Hardness=ScalarParameter() {defaultValue=0.9, group="3) Sphere Mask", parameterName=Hardness}
  ) {attenuationRadius=100, hardnessPercent=1}
  MaterialFunctionCall(Alpha Threshold=LinearInterpolate(A=@n22, B=Multiply(A=@n22, B=@n24), Alpha=ScalarParameter() {defaultValue=1, group="3) Sphere Mask", parameterName=Use Sphere Mask})) {materialFunction=&DitherTemporalAA_Fortnite}
 ]
}
