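# Niagara system for the VectorSpark emote that flows cyan energy along the character's arms and hands.
# Two ribbon emitters link six bone-sampled points by spawn group, nudged toward the camera, with the Low one's tint and alpha driven by System charge and stage values.
# Burst-spawned hand meshes, fire sprites and light-emitting glows follow sampled hand bones.
# PPV_color sprites use distance-based camera offset, and several emitters scale brightness by time of day.
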NiagaraSystem /BRCosmetics/Effects/Fort_Effects/Emotes/VectorSpark/NS_Emote_VectorSpark_EnergyFlow
dependencies {
 usedDataInterfaces [
  &NiagaraDataInterfaceCurve
  &NiagaraDataInterfaceCamera
  &NiagaraDataInterfaceArrayFloat3
  &NiagaraDataInterfaceSkeletalMesh
 ]
 usedDynamicInputs []
 usedModules [
  &self:SystemSpawnScript.SystemScriptSource.SystemScriptGraph.NiagaraNodeAssignment_8.SetVariables_#1
  &self:SystemSpawnScript.SystemScriptSource.SystemScriptGraph.NiagaraNodeAssignment_6.SetVariables_#2
  &SystemState
 ]
}
emitters {
 Arm_Ribbons_Low {
  emitterUpdateScript [
   &EmitterState()
   &SpawnBurst_Instantaneous(Spawn Count=6, Spawn Time=0.15, Use Spawn Probability=false, UseLoopCountLimit=false)
  ]
  particleSpawnScript [
   &FN_SetSpawnGroup(Number of Spawn Groups=6)
   &InitializeParticle(
    Color=<rgba 0 1.3368 2 0.1>
    Lifetime=2
    Ribbon Width=45
    Ribbon Width Mode=::NewEnumerator1
    Sprite Size Mode=::NewEnumerator3
    Uniform Sprite Size=100
    UsePositionOffset=false
    Write Material Random=true
   )
   &CameraOffset(Camera Offset Amount=100)
  ]
  particleUpdateScript [
   &ParticleState(Kill Particles When Lifetime Has Elapsed=false, Let Infinitely Lived Particles Die When Emitter Deactivates=true)
   &SkeletalMeshLocation(
    Bone / Socket Index=$Particles.SpawnGroup
    Bone Sampling Mode=::NewEnumerator4
    MaskSpawnGroup=false
    Orientation Sampling=::NewEnumerator1
    Sampled Position Offset=&SelectVectorFromArray(
     Direct Array Index=$Particles.SpawnGroup
     Vector Selection Array=DataInterface() {propertyValues {FloatData [<-5 0 0>, <0 -3 0>, <0 10 10>, <5 0 -5>, <0 3 0>, <0 0 0>]}}
    )
    Skeletal Mesh=$System.Skeletal Mesh_Ribbon
    Use Sampled Mesh Attribute as Ribbon ID=false
   )
   /SetVariables(Particles.RibbonLinkOrder:NiagaraFloat=&MakeFloatFromInt(Int=$Particles.SpawnGroup)) {isSetParameters=true, module=SetVariables_#3}
   /SetVariables(Particles.Position=&AddVectorToPosition(
     Position=$Particles.Position
     Vector=&Multiply_VectorByFloat(
      Vector=&NormalizeVector(
       Vector To Normalize=&SubtractVectorFromPosition(
        Position=&ReturnCameraPosition(Coordinate Space:ENiagaraCoordinateSpace=::Simulation, Camera Query:NiagaraDataInterfaceCamera=DataInterface())
        Vector=Linked() {linkedVariable {name=Particles.Position, type=NiagaraPosition}}
       )
       Safe Normalize=true
       Safe Fallback Threshold=0
       Safe Fallback Vector=<0 0 0.001>
      )
      Float=15
     )
    )) {isSetParameters=true, module=SetVariables_#4}
   &ScaleColor(
    Scale Alpha=&Lerp_Float(A=1, B=0, Alpha=$System.StageChange)
    Scale RGB=&MakeVectorFromLinearColorRGB(Color=&LerpLinearColors(StartColor=<rgba 1 1 1 1>, EndColor=<rgba 1 0 0.1469 1>, LerpFactor=$System.IsCharging))
    ScaleA=true
    ScaleRGB=true
   )
   &ScaleRibbonWidth(Ribbon Width Scale=&FloatFromCurve(FloatCurve=DataInterface() {propertyValues {Curve {keys [<at 0 1>,<at 1 0.8>]}}}, CurveIndex=$Particles.RibbonLinkOrder, Scale Curve=0.6))
   &DynamicMaterialParameters(
    Index 0 Param 1=0.25
    Index 0 Param 2=0.5
    Index 0 Param 4=0.25
    Index 1 Param 1=&FloatFromCurve(FloatCurve=DataInterface() {propertyValues {Curve {keys [
         {arriveTangent=-19.5507, interpMode=RCIM_Cubic, time=0, value=2}
         {arriveTangent=-2.25, leaveTangent=-2.25, time=0.2, value=0.2}
         {arriveTangent=2.25, interpMode=RCIM_Cubic, leaveTangent=2.25, time=0.6, value=0.2}
         {arriveTangent=11.5048, interpMode=RCIM_Cubic, leaveTangent=11.5048, tangentMode=RCTM_User, time=1, value=2}
        ]}}})
    Module.Param0WriteEnabled=true
    Module.Param0WriteEnabled1=true
    Module.Param1WriteEnabled=true
    Module.Param1WriteEnabled1=false
    Module.Param2WriteEnabled=false
    Module.Param2WriteEnabled1=false
    Module.Param3WriteEnabled=true
    Module.Param3WriteEnabled1=false
    Write Parameter Index 1=true
   )
   /SetVariables(Particles.RibbonID=&self:ScratchDynamicInput_05(SpawnGroup:NiagaraFloat=&MakeFloatFromInt(Int=$Particles.SpawnGroup))) {isSetParameters=true, module=SetVariables_#5}
  ]
  rendererClasses [&NiagaraRibbonRendererProperties]
  renderers [{rendererClass=&NiagaraRibbonRendererProperties}]
 }
 Arms_Ribbons {
  emitterUpdateScript [
   &EmitterState()
   &SpawnBurst_Instantaneous(Spawn Count=6, Spawn Time=0.15, Use Spawn Probability=true, UseLoopCountLimit=false)
  ]
  particleSpawnScript [
   &FN_SetSpawnGroup(Number of Spawn Groups=6)
   &InitializeParticle(
    Color=<rgba 0 1.3368 2 0.1>
    Lifetime=2
    Ribbon Width=40
    Ribbon Width Mode=::NewEnumerator1
    Sprite Size Mode=::NewEnumerator3
    Uniform Sprite Size=100
    UsePositionOffset=false
    Write Material Random=true
   )
   &CameraOffset(Camera Offset Amount=100)
  ]
  particleUpdateScript [
   &ParticleState(Kill Particles When Lifetime Has Elapsed=false, Let Infinitely Lived Particles Die When Emitter Deactivates=true)
   &SkeletalMeshLocation(
    Bone / Socket Index=$Particles.SpawnGroup
    Bone Sampling Mode=::NewEnumerator4
    MaskSpawnGroup=false
    Orientation Sampling=::NewEnumerator1
    Sampled Position Offset=&SelectVectorFromArray(
     Direct Array Index=$Particles.SpawnGroup
     Vector Selection Array=DataInterface() {propertyValues {FloatData [<-15 3 0>, <0 -3 0>, <10 10 0>, <15 0 -5>, <0 5 0>, <0 0 0>]}}
    )
    Skeletal Mesh=$System.Skeletal Mesh_Ribbon
    Use Sampled Mesh Attribute as Ribbon ID=false
   )
   /SetVariables(Particles.RibbonLinkOrder:NiagaraFloat=&MakeFloatFromInt(Int=$Particles.SpawnGroup)) {isSetParameters=true, module=SetVariables_#3}
   /SetVariables(Particles.Position=&AddVectorToPosition(
     Position=$Particles.Position
     Vector=&Multiply_VectorByFloat(
      Vector=&NormalizeVector(
       Vector To Normalize=&SubtractVectorFromPosition(
        Position=&ReturnCameraPosition(Coordinate Space:ENiagaraCoordinateSpace=::Simulation, Camera Query:NiagaraDataInterfaceCamera=DataInterface())
        Vector=Linked() {linkedVariable {name=Particles.Position, type=NiagaraPosition}}
       )
       Safe Normalize=true
       Safe Fallback Threshold=0
       Safe Fallback Vector=<0 0 0.001>
      )
      Float=10
     )
    )) {isSetParameters=true, module=SetVariables_#4}
   &ScaleRibbonWidth(Ribbon Width Scale=&FloatFromCurve(
     FloatCurve=DataInterface() {propertyValues {Curve {keys [<at 0 1 RCIM_Cubic>, {arriveTangent=-0.7409, interpMode=RCIM_Cubic, leaveTangent=-0.7409, tangentMode=RCTM_User, time=1, value=0.75}]}}}
     CurveIndex=$Particles.RibbonLinkOrder
     Scale Curve=&FloatFromCurve(FloatCurve=DataInterface() {propertyValues {Curve {keys [
          {arriveTangent=2.6108, interpMode=RCIM_Cubic, leaveTangent=2.6108, tangentMode=RCTM_User, time=0, value=0}
          {interpMode=RCIM_Cubic, leaveTangent=0.5, time=0.3, value=1}
         ]}}})
    ))
   &DynamicMaterialParameters(
    Index 0 Param 1=0.25
    Index 0 Param 2=0.5
    Index 0 Param 4=0.25
    Index 1 Param 1=&FloatFromCurve(FloatCurve=DataInterface() {propertyValues {Curve {keys [
         {arriveTangent=-19.5507, interpMode=RCIM_Cubic, time=0, value=2}
         {arriveTangent=-2.25, leaveTangent=-2.25, time=0.2, value=0.2}
         {arriveTangent=2.25, interpMode=RCIM_Cubic, leaveTangent=2.25, time=0.6, value=0.2}
         {arriveTangent=11.5048, interpMode=RCIM_Cubic, leaveTangent=11.5048, tangentMode=RCTM_User, time=1, value=2}
        ]}}})
    Module.Param0WriteEnabled=true
    Module.Param0WriteEnabled1=true
    Module.Param1WriteEnabled=true
    Module.Param1WriteEnabled1=false
    Module.Param2WriteEnabled=false
    Module.Param2WriteEnabled1=false
    Module.Param3WriteEnabled=true
    Module.Param3WriteEnabled1=false
    Write Parameter Index 1=true
   )
   /SetVariables(Particles.RibbonID=&self:ScratchDynamicInput_05(SpawnGroup:NiagaraFloat=&MakeFloatFromInt(Int=$Particles.SpawnGroup))) {isSetParameters=true, module=SetVariables_#5}
  ]
  rendererClasses [&NiagaraRibbonRendererProperties]
  renderers [{rendererClass=&NiagaraRibbonRendererProperties}]
 }
 Glow {
  emitterUpdateScript [
   &EmitterState()
   &SpawnBurst_Instantaneous(Spawn Count=2, Spawn Time=0.125, Use Spawn Probability=false, UseLoopCountLimit=false)
  ]
  particleSpawnScript [
   &PartitionParticles()
   &InitializeParticle(Lifetime=2, Sprite Size Mode=::NewEnumerator3, Uniform Sprite Size=70, UsePositionOffset=false, Write Material Random=true)
  ]
  particleUpdateScript [
   &ParticleState()
   &SkeletalMeshLocation(
    Bone / Socket Index=$Particles.PartitionParticles.Partition
    Bone Sampling Mode=::NewEnumerator4
    MaskSpawnGroup=false
    Sampled Position Offset=&Lerp_Vector(A=<10 0 0>, B=<-10 0 0>, Alpha=&MakeFloatFromInt(Int=$Particles.PartitionParticles.Partition))
    Skeletal Mesh=$System.Skeletal Mesh_hand
   )
   &ScaleColor(
    Scale Alpha=&Multiply_Float(
     A=&FloatFromCurve(
      FloatCurve=DataInterface() {propertyValues {Curve {keys [
          {arriveTangent=0.5594, interpMode=RCIM_Cubic, leaveTangent=0.5594, tangentMode=RCTM_User, time=0, value=0}
          {arriveTangent=1.25, leaveTangent=1.25, time=0.1408, value=1}
          {arriveTangent=-1.0394, interpMode=RCIM_Cubic, leaveTangent=-1.0394, tangentMode=RCTM_User, time=0.7, value=1}
          {arriveTangent=2.1479, interpMode=RCIM_Cubic, leaveTangent=2.1479, tangentMode=RCTM_User, time=1, value=0}
         ]}}}
      Scale Curve=0.25
     )
     B=&FNTimeOfDayScalar(Night Scale=0.5, Morning Scale=1, Evening Scale=1)
    )
    Scale RGB=&MakeVectorFromLinearColorRGB(Color=<rgba 0 0.3417 0.4948 1>)
    ScaleA=true
    ScaleRGB=true
   )
   &FN_Light_Attributes(
    Color=<rgba 0 13.8113 20 1>
    Create Color Attribute=true
    Create Exponent Attribute=true
    Create Position Attribute=true
    Create Radius Attribute=true
    Exponent=3
    Position=&AddVectorToPosition(
     Position=$Particles.Position
     Vector=&Lerp_Vector(
      A=<0 -10 20>
      B=<0 0 0>
      Alpha=&ClampFloat(Float=&MakeCustomFloatFromBool(
        Bool=&SetBoolByFloatComparison(A=&VectorLength(Vector=$Engine.Owner.Velocity), B=100)
        True Float=1
        False Float=0
       ))
     )
    )
    Radius=&FloatFromCurve(
     FloatCurve=DataInterface() {propertyValues {Curve {keys [
         {arriveTangent=0.5594, interpMode=RCIM_Cubic, leaveTangent=0.5594, tangentMode=RCTM_User, time=0, value=0}
         {arriveTangent=1.25, leaveTangent=1.25, time=0.1408, value=1}
         {arriveTangent=-1.0394, interpMode=RCIM_Cubic, leaveTangent=-1.0394, tangentMode=RCTM_User, time=0.8, value=1}
         {arriveTangent=2.1479, interpMode=RCIM_Cubic, leaveTangent=2.1479, tangentMode=RCTM_User, time=1, value=0}
        ]}}}
     Scale Curve=50
    )
   )
   &ScaleSpriteSize(Uniform Curve Sprite Scale=DataInterface() {propertyValues {Curve {keys [
        {arriveTangent=0.2096, interpMode=RCIM_Cubic, leaveTangent=0.2096, tangentMode=RCTM_User, time=0, value=0}
        {arriveTangent=1.1179, interpMode=RCIM_Cubic, leaveTangent=1.1179, tangentMode=RCTM_User, time=0.2042, value=0.6228}
        {arriveTangent=0.0224, interpMode=RCIM_Cubic, leaveTangent=0.0224, tangentMode=RCTM_User, time=1, value=1}
       ]}}})
  ]
  rendererClasses [&NiagaraSpriteRendererProperties, &NiagaraLightRendererProperties]
  renderers [
   {rendererClass=&NiagaraSpriteRendererProperties}
   {rendererClass=&NiagaraLightRendererProperties, rendererIndex=1}
  ]
 }
 Hands_Blue {
  emitterUpdateScript [
   &EmitterState()
   &SpawnBurst_Instantaneous(Spawn Time=0.15, Use Spawn Probability=false, UseLoopCountLimit=false)
   &SpawnBurst_Instantaneous(Spawn Group=1, Spawn Time=0.15, Use Spawn Probability=false, UseLoopCountLimit=false) {module=SpawnBurst_Instantaneous001}
   /SetVariables(Emitter.ColorHigh:LinearColor=<rgba 0 0.6684 1 0.1>, Emitter.ColorLow:LinearColor=<rgba 0 0.6684 1 0.1>) {isSetParameters=true}
  ]
  particleSpawnScript [
   &FN_SetSpawnGroup(Number of Spawn Groups=2)
   &InitializeParticle(
    Color=<rgba 0 1.3368 2 0.1>
    Lifetime=2
    Mesh Scale=<0.1 0.15 0.15>
    Mesh Scale Mode=::NewEnumerator4
    Position Offset=<7 0 0>
    Position Offset Coordinate Space=::Local
    UsePositionOffset=true
    Write Material Random=true
   )
   &InitialMeshOrientation(
    Mesh Orientation Mode=::NewEnumerator2
    Orientation Vector=&Lerp_Vector(A=<-1 0 0>, B=<1 0 0>, Alpha=&MakeFloatFromInt(Int=$Particles.SpawnGroup))
    Rotation=&SelectVectorFromArray(Direct Array Index=$Particles.SpawnGroup, Vector Selection Array=DataInterface() {propertyValues {FloatData [<0.5 0 -0.05>, <0.5 0 0.05>]}})
    Use Rotation Vector=true
   )
  ]
  particleUpdateScript [
   &ParticleState()
   &SkeletalMeshLocation(
    Bone / Socket Index=$Particles.SpawnGroup
    Bone Sampling Mode=::NewEnumerator4
    MaskSpawnGroup=false
    Sampled Position Offset=&Lerp_Vector(A=<15 -5 0>, B=<-15 5 0>, Alpha=&MakeFloatFromInt(Int=&SpawnGroup()))
    Skeletal Mesh=$System.Skeletal Mesh_hand
   )
   &ScaleColor(
    Color Value To Scale=$Emitter.ColorHigh
    Scale Alpha=&Multiply_Float(A=0.1, B=&FNTimeOfDayScalar(Night Scale=0.7, Morning Scale=1, Evening Scale=1))
    ScaleA=true
    ScaleRGB=true
   )
   &ScaleMeshSize(Scale Factor=&Multiply_VectorByFloat(
     Vector=<1 1 1>
     Float=&FloatFromCurve(FloatCurve=DataInterface() {propertyValues {Curve {keys [
          {arriveTangent=8.2074, interpMode=RCIM_Cubic, leaveTangent=8.2074, tangentMode=RCTM_User, time=0, value=0.2}
          {arriveTangent=1.3333, leaveTangent=1.3333, time=0.2, value=1}
          {arriveTangent=-1.25, interpMode=RCIM_Cubic, leaveTangent=-1.25, time=0.6, value=1}
          {arriveTangent=-9.3674, interpMode=RCIM_Cubic, leaveTangent=-9.3674, tangentMode=RCTM_User, time=1, value=0}
         ]}}})
    ))
   /SetVariables(Particles.ColorBlack:LinearColor=&MultiplyLinearColor(Linear Color=$Particles.Color, Scale Factor=<1 1 1 8>)) {isSetParameters=true}
  ]
  rendererClasses [&NiagaraMeshRendererProperties]
  renderers [
   {rendererClass=&NiagaraMeshRendererProperties, rendererIndex=0}
   {rendererClass=&NiagaraMeshRendererProperties, rendererIndex=1}
  ]
 }
 Hands_Fire {
  emitterUpdateScript [
   &EmitterState()
   &SpawnBurst_Instantaneous(Spawn Count=2, Spawn Time=0.15, Use Spawn Probability=false, UseLoopCountLimit=false)
  ]
  particleSpawnScript [
   &PartitionParticles()
   &InitializeParticle(
    Color=<rgba 0 1.0026 1.5 0.1>
    Lifetime=2.1
    Mesh Scale=<0.65 0.65 0.65>
    Mesh Scale Mode=::NewEnumerator4
    Sprite Size Mode=::NewEnumerator3
    Uniform Sprite Size=40
    UsePositionOffset=false
    Write Material Random=true
   )
   &CameraOffset(Camera Offset Amount=10)
   &InitialMeshOrientation(Mesh Orientation Mode=::NewEnumerator2, Rotation=<0 -0.25 0>, Use Rotation Vector=true)
  ]
  particleUpdateScript [
   &ParticleState()
   &SkeletalMeshLocation(
    Bone / Socket Index=$Particles.PartitionParticles.Partition
    Bone Sampling Mode=::NewEnumerator4
    MaskSpawnGroup=false
    Orientation Sampling=::NewEnumerator1
    Sampled Position Offset=&Lerp_Vector(A=<10 0 0>, B=<-10 0 0>, Alpha=&MakeFloatFromInt(Int=$Particles.PartitionParticles.Partition))
    Skeletal Mesh=$System.Skeletal Mesh_hand
   )
   &DynamicMaterialParameters(
    Index 0 Param 1=&RandomRangeFloat(Minimum=0, Maximum=0.5)
    Index 0 Param 2=&RandomRangeFloat(Minimum=0, Maximum=0.25)
    Index 0 Param 4=&RandomRangeFloat(Minimum=0, Maximum=0.25)
    Index 1 Param 1=&FloatFromCurve(FloatCurve=DataInterface() {propertyValues {Curve {keys [
         {arriveTangent=-6.4125, interpMode=RCIM_Cubic, leaveTangent=-6.4125, tangentMode=RCTM_User, time=0, value=2}
         {arriveTangent=-2.25, leaveTangent=-2.25, time=0.25, value=0.2}
         {arriveTangent=5.0667, interpMode=RCIM_Cubic, leaveTangent=5.0667, time=0.6, value=0.2}
         {arriveTangent=20.5791, interpMode=RCIM_Cubic, leaveTangent=20.5791, tangentMode=RCTM_User, time=1, value=4}
        ]}}})
    Module.Param0WriteEnabled=true
    Module.Param0WriteEnabled1=true
    Module.Param1WriteEnabled=true
    Module.Param1WriteEnabled1=true
    Module.Param2WriteEnabled=false
    Module.Param2WriteEnabled1=true
    Module.Param3WriteEnabled=true
    Module.Param3WriteEnabled1=true
    Write Parameter Index 1=true
   )
  ]
  rendererClasses [&NiagaraMeshRendererProperties]
  renderers [{rendererClass=&NiagaraMeshRendererProperties, rendererIndex=0}]
 }
 Hands_Fire_Sprite {
  emitterUpdateScript [
   &EmitterState()
   &SpawnBurst_Instantaneous(Spawn Count=2, Spawn Time=0.125, Use Spawn Probability=false, UseLoopCountLimit=false)
  ]
  particleSpawnScript [
   &PartitionParticles()
   &InitializeParticle(
    Color=<rgba 0 1.0026 1.5 0.1>
    Lifetime=2.1
    Mesh Scale=<0.5 0.5 0.5>
    Mesh Scale Mode=::NewEnumerator4
    Sprite Size Mode=::NewEnumerator3
    Uniform Sprite Size=40
    UsePositionOffset=false
    Write Material Random=true
   )
   &CameraOffset(Camera Offset Amount=10)
  ]
  particleUpdateScript [
   &ParticleState()
   &ScaleSpriteSize(Uniform Curve Sprite Scale=DataInterface() {propertyValues {Curve {keys [
        {arriveTangent=8.2074, interpMode=RCIM_Cubic, leaveTangent=8.2074, tangentMode=RCTM_User, time=0, value=0.2}
        {arriveTangent=1.3333, leaveTangent=1.3333, time=0.2, value=1}
        {arriveTangent=-5.1542, interpMode=RCIM_Cubic, leaveTangent=-5.1542, tangentMode=RCTM_User, time=0.6, value=1}
        {arriveTangent=-0.3511, interpMode=RCIM_Cubic, leaveTangent=-0.3511, tangentMode=RCTM_User, time=1, value=0}
       ]}}})
   &SkeletalMeshLocation(
    Bone / Socket Index=$Particles.PartitionParticles.Partition
    Bone Sampling Mode=::NewEnumerator4
    MaskSpawnGroup=false
    Orientation Sampling=::NewEnumerator1
    Sampled Position Offset=&Lerp_Vector(A=<10 0 0>, B=<-10 0 0>, Alpha=&MakeFloatFromInt(Int=$Particles.PartitionParticles.Partition))
    Skeletal Mesh=$System.Skeletal Mesh_hand
   )
   &DynamicMaterialParameters(Module.Param0WriteEnabled=true, Module.Param1WriteEnabled=false, Module.Param2WriteEnabled=false, Module.Param3WriteEnabled=false)
  ]
  rendererClasses [&NiagaraSpriteRendererProperties]
  renderers [{rendererClass=&NiagaraSpriteRendererProperties}]
 }
 PPV_color {
  emitterUpdateScript [
   &EmitterState()
   &SpawnBurst_Instantaneous(Spawn Count=2, Use Spawn Probability=false, UseLoopCountLimit=false)
  ]
  particleSpawnScript [
   &PartitionParticles()
   &InitializeParticle(Lifetime=3.5, Sprite Size Mode=::NewEnumerator3, Uniform Sprite Size=32.5, UsePositionOffset=false, Write Material Random=true)
  ]
  particleUpdateScript [
   &ParticleState()
   &SkeletalMeshLocation(
    Bone / Socket Index=$Particles.PartitionParticles.Partition
    Bone Sampling Mode=::NewEnumerator4
    MaskSpawnGroup=false
    Skeletal Mesh=$System.Skeletal Mesh_hand
   )
   &CameraOffset(Camera Offset Amount=&FloatFromCurve(
     FloatCurve=DataInterface() {propertyValues {Curve {keys [<at 0 0>,<at 512 95>]}}}
     CurveIndex=&VectorLength(Vector=&SubtractVectorFromPosition(
       Position=&ReturnCameraPosition(Coordinate Space:ENiagaraCoordinateSpace=::Simulation, Camera Query:NiagaraDataInterfaceCamera=DataInterface())
       Vector=Linked() {linkedVariable {name=Particles.Position, type=NiagaraPosition}}
      ))
     Scale Curve=2
    ))
   &ScaleColor(
    Scale RGB=&Multiply_VectorByFloat(
     Vector=&MakeVectorFromLinearColorRGB(Color=<rgba 0 0.6667 1 1>)
     Float=&FNTimeOfDayScalar(Night Scale=0.5, Morning Scale=1, Evening Scale=1)
    )
    ScaleA=true
    ScaleRGB=true
   )
   &DynamicMaterialParameters(
    Index 0 Param 1=1
    Module.Param0WriteEnabled=true
    Module.Param1WriteEnabled=false
    Module.Param2WriteEnabled=false
    Module.Param3WriteEnabled=false
   )
  ]
  rendererClasses [&NiagaraSpriteRendererProperties]
  renderers [
   {rendererClass=&NiagaraSpriteRendererProperties}
   {rendererClass=&NiagaraSpriteRendererProperties, rendererIndex=1}
  ]
 }
}
summary {
 systemName=NS_Emote_VectorSpark_EnergyFlow
 emitters [
  {emitterName=PPV_color, rendererClasses [&NiagaraSpriteRendererProperties]}
  {emitterName=Glow, rendererClasses [&NiagaraSpriteRendererProperties, &NiagaraLightRendererProperties]}
  {emitterName=Hands_Fire, rendererClasses [&NiagaraMeshRendererProperties]}
  {emitterName=Hands_Blue, rendererClasses [&NiagaraMeshRendererProperties]}
  {emitterName=Hands_Fire_Sprite, rendererClasses [&NiagaraSpriteRendererProperties]}
  {emitterName=Arm_Ribbons_Low, rendererClasses [&NiagaraRibbonRendererProperties]}
  {emitterName=Arms_Ribbons, rendererClasses [&NiagaraRibbonRendererProperties]}
 ]
 userVariables []
}
