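# Niagara effect for the VectorSpark emote (Shadow V2), layering a light cone, glow, wave and mask on a skeletal mesh.
# The cone mesh and three SubUV glow sprites spawn at an offset and align along the line to sampled bone positions, with time-of-day tinting.
# About half a second in, short-lived wave and mask meshes follow sampled bone positions.
# Curves drive fades, scaling, cone WPO and the mask's material parameter.
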NiagaraSystem /BRCosmetics/Effects/Fort_Effects/Emotes/VectorSpark/NS_Emote_VectorSpark_Shadow_V2
dependencies {
 usedDataInterfaces [
  &NiagaraDataInterfaceSpriteRendererInfo
  &NiagaraDataInterfaceCurve
  &NiagaraDataInterfaceVector2DCurve
  &NiagaraDataInterfaceSkeletalMesh
 ]
 usedDynamicInputs [&MakeCustomVectorFromBool]
 usedModules [
  &self:SystemSpawnScript.SystemScriptSource.SystemScriptGraph.NiagaraNodeAssignment_0.SetVariables_#1
  &SystemState
 ]
}
emitters {
 LightCone_FromHand {
  emitterUpdateScript [
   &EmitterState()
   &SpawnBurst_Instantaneous(Use Spawn Probability=false, UseLoopCountLimit=true)
  ]
  particleSpawnScript [
   &InitializeParticle(
    Color=&FNTimeOfDayColor(
     Night Color=<rgba 1.5833 1.9063 2 0.075>
     Morning Color=<rgba 1.5833 1.9063 2 0.1>
     Day Color=<rgba 1.5833 1.9063 2 0.1>
     Evening Color=<rgba 1.5833 1.9063 2 0.1>
    )
    Lifetime=4.6
    Mesh Scale Mode=::NewEnumerator3
    Mesh Uniform Scale=2
    UsePositionOffset=false
    Write Material Random=true
   )
   &SystemLocation(MaskSpawnGroup=false, Offset=<-25 -50 85>)
  ]
  particleUpdateScript [
   &ParticleState()
   &SkeletalMeshLocation(
    Bone Sampling Mode=::NewEnumerator4
    MaskSpawnGroup=false
    Orientation Sampling=::NewEnumerator1
    Position Sampling=::NewEnumerator2
    Sampled Position Offset Coordinate Space:ENiagaraSkelMeshTransforms=::NewEnumerator2
    Skeletal Mesh=$System.Skeletal Mesh
   )
   /SetVariables(Particles.MeshOrientation=&MultiplyQuaternion(
     Quaternion A=&FindQuatBetweenTwoVectors(
      From Vector=<1 0 0>
      To Vector=&NormalizeVector(
       Vector To Normalize=&SubtractVectorFromPosition(Position=$Particles.SkeletalMeshLocation.SampledPosition, Vector=Linked() {linkedVariable {name=Particles.Position, type=NiagaraPosition}})
       Safe Normalize=true
       Safe Fallback Threshold=0
       Safe Fallback Vector=<0 0 0.001>
      )
     )
     Quaternion B=&FindQuatBetweenTwoVectors(From Vector=<0 0 1>, To Vector=<1 0 0>)
    )) {isSetParameters=true}
   &ScaleColor(
    Scale Alpha=&FloatFromCurve(FloatCurve=DataInterface() {propertyValues {Curve {keys [
         {arriveTangent=17.9214, interpMode=RCIM_Cubic, leaveTangent=17.9214, tangentMode=RCTM_User, time=0, value=0}
         {arriveTangent=1.1111, leaveTangent=1.1111, time=0.2, value=1}
         {arriveTangent=-1.25, interpMode=RCIM_Cubic, leaveTangent=-1.25, time=0.9, value=1}
         {arriveTangent=-29.2594, interpMode=RCIM_Cubic, leaveTangent=-29.2594, tangentMode=RCTM_User, time=1, value=0}
        ]}}})
    ScaleA=true
    ScaleRGB=true
   )
   /SetVariables(Particles.WPO:NiagaraFloat=&FloatFromCurve(
     FloatCurve=DataInterface() {propertyValues {Curve {keys [
         {arriveTangent=17.9214, interpMode=RCIM_Cubic, time=0, value=0}
         {arriveTangent=1.1111, leaveTangent=1.1111, time=0.075, value=1}
         {arriveTangent=-1.0811, interpMode=RCIM_Cubic, leaveTangent=-1.0811, time=0.9, value=1}
         {arriveTangent=-29.2594, interpMode=RCIM_Cubic, leaveTangent=-29.2594, tangentMode=RCTM_User, time=1, value=0}
        ]}}}
     Scale Curve=60
    )) {isSetParameters=true}
  ]
  rendererClasses [&NiagaraMeshRendererProperties]
  renderers [{rendererClass=&NiagaraMeshRendererProperties, rendererIndex=0}]
 }
 Mask001 {
  emitterUpdateScript [
   &EmitterState()
   &SpawnBurst_Instantaneous(Spawn Time=0.5, Use Spawn Probability=false, UseLoopCountLimit=false)
  ]
  particleSpawnScript [
   &InitializeParticle(Lifetime=0.7, Mesh Scale Mode=::NewEnumerator3, Sprite Rotation Mode=::NewEnumerator2, UsePositionOffset=false, Write Material Random=true)
   &InitialMeshOrientation(Mesh Orientation Mode=::NewEnumerator2, Use Rotation Vector=true)
   &CameraOffset(Camera Offset Amount=1)
  ]
  particleUpdateScript [
   &ParticleState()
   &SkeletalMeshLocation(Bone Sampling Mode=::NewEnumerator4, MaskSpawnGroup=false, Sampled Position Offset=<-14 0 -1>, Skeletal Mesh=$System.Skeletal Mesh)
   &ScaleColor(
    Scale Alpha=&FloatFromCurve(FloatCurve=DataInterface() {propertyValues {Curve {keys [
         {arriveTangent=0.2119, interpMode=RCIM_Cubic, leaveTangent=0.2119, tangentMode=RCTM_User, time=0, value=-0.3}
         {arriveTangent=1.078, interpMode=RCIM_Cubic, leaveTangent=-3.0773, tangentMode=RCTM_Break, time=0.9, value=0.9}
         {arriveTangent=-46.2165, interpMode=RCIM_Cubic, leaveTangent=-46.2165, tangentMode=RCTM_User, time=1, value=-0.3}
        ]}}})
    ScaleA=true
    ScaleRGB=true
   )
   &DynamicMaterialParameters(
    Index 0 Param 1=&FloatFromCurve(FloatCurve=DataInterface() {propertyValues {Curve {keys [{interpMode=RCIM_Cubic, tangentMode=RCTM_User, time=0.5, value=0}, {interpMode=RCIM_Cubic, tangentMode=RCTM_User, time=0.9, value=1}]}}})
    Module.Param0WriteEnabled=true
    Module.Param1WriteEnabled=true
    Module.Param2WriteEnabled=true
    Module.Param3WriteEnabled=true
   )
  ]
  rendererClasses [&NiagaraMeshRendererProperties]
  renderers [{rendererClass=&NiagaraMeshRendererProperties, rendererIndex=0}]
 }
 SubUV_ConeGlowSkeletal {
  emitterUpdateScript [
   &EmitterState()
   &SpawnBurst_Instantaneous(Spawn Count=3, Use Spawn Probability=false, UseLoopCountLimit=false)
  ]
  particleSpawnScript [
   &InitializeParticle(
    Color=&FNTimeOfDayColor(
     Night Color=<rgba 1.5833 1.9063 2 0.025>
     Morning Color=<rgba 1.5833 1.9063 2 0.025>
     Day Color=<rgba 1.5833 1.9063 2 0.075>
     Evening Color=<rgba 1.5833 1.9063 2 0.05>
    )
    Lifetime=4.6
    Position Offset=<-25 -50 85>
    Sprite Size Max=<70 165>
    Sprite Size Min=<70 150>
    Sprite Size Mode=::NewEnumerator2
    Sprite UV Mode=::NewEnumerator4
    Sprite UV Scale:Vector2f=<1 -1>
    UsePositionOffset=true
    Write Material Random=true
   )
   &SubUVAnimation(
    End Frame Range Override=2
    Evaluation Type=::NewEnumerator0
    "Sprite Renderer "=DataInterface() {propertyValues {SpriteRenderer=&self:NiagaraEmitter_3.NiagaraSpriteRendererProperties_0}}
    SubUV Animation Mode:ENiagara_SubUVLookupModeV2=::NewEnumerator2
    UseEndFrame=true
    UseStartFrame=false
   )
   &CameraOffset(Camera Offset Amount=5)
  ]
  particleUpdateScript [
   &ParticleState()
   &SkeletalMeshLocation(Bone Sampling Mode=::NewEnumerator4, MaskSpawnGroup=false, Skeletal Mesh=$System.Skeletal Mesh, Write To Particle Position=false)
   &SpriteFacingAndAlignment(
    Sprite Alignment=&NormalizeVector(
     Vector To Normalize=&Subtract_Position(A=$Particles.Position, B=$Particles.SkeletalMeshLocation.SampledPosition)
     Safe Normalize=true
     Safe Fallback Threshold=0
     Safe Fallback Vector=<0 0 0.001>
    )
    WriteAlignment=true
    WriteFacing=false
   )
   &ScaleColor(
    Scale Alpha=&FloatFromCurve(FloatCurve=DataInterface() {propertyValues {Curve {keys [
         {arriveTangent=17.9214, interpMode=RCIM_Cubic, leaveTangent=17.9214, tangentMode=RCTM_User, time=0, value=0}
         {arriveTangent=1.1111, leaveTangent=1.1111, time=0.2, value=1}
         {arriveTangent=-1.25, interpMode=RCIM_Cubic, leaveTangent=-1.25, time=0.9, value=1}
         {arriveTangent=-29.2594, interpMode=RCIM_Cubic, leaveTangent=-29.2594, tangentMode=RCTM_User, time=1, value=0}
        ]}}})
    ScaleA=true
    ScaleRGB=true
   )
   &ScaleSpriteSize(
    Non-Uniform Curve Sprite Scale=DataInterface() {propertyValues {
      XCurve {keys [
        <at 0 0 RCIM_Cubic>
        <at 0.075 1>
        {arriveTangent=-1.0811, interpMode=RCIM_Cubic, leaveTangent=-1.0811, time=0.9, value=1}
        {arriveTangent=-19.5899, interpMode=RCIM_Cubic, leaveTangent=-19.5899, tangentMode=RCTM_User, time=1, value=0}
       ]}
      YCurve {keys [{tangentMode=RCTM_User, time=0.1, value=1}]}
     }}
    Scale Sprite Size Mode=::NewEnumerator3
   )
  ]
  rendererClasses [&NiagaraSpriteRendererProperties]
  renderers [{rendererClass=&NiagaraSpriteRendererProperties}]
 }
 Wave001 {
  emitterUpdateScript [
   &EmitterState()
   &SpawnBurst_Instantaneous(Spawn Time=0.55, Use Spawn Probability=false, UseLoopCountLimit=false)
  ]
  particleSpawnScript [
   &InitializeParticle(
    Color=&RandomRangeLinearColor(Minimum=<rgba 0 0 0 2.5>, Maximum=<rgba 0 0 0 3>)
    Lifetime=0.6
    Mesh Scale Mode=::NewEnumerator3
    Mesh Uniform Scale=0.67
    UsePositionOffset=false
    Write Material Random=true
   )
   &InitialMeshOrientation(Mesh Axis To Orient=<0 0 1>, Mesh Orientation Mode=::NewEnumerator2, Rotation=<0.25 0 0>, Use Rotation Vector=true)
   &DynamicMaterialParameters(
    Index 0 Param 1=2
    Index 0 Param 2=&RandomRangeFloat(Minimum=-1.5, Maximum=-2.25)
    Module.Param0WriteEnabled=true
    Module.Param1WriteEnabled=true
    Module.Param2WriteEnabled=true
    Module.Param3WriteEnabled=true
   )
   &CameraOffset(Camera Offset Amount=10)
  ]
  particleUpdateScript [
   &ParticleState()
   &UpdateMeshOrientation(Rotation Rate=&RandomRangeFloat(Minimum=0.1, Maximum=0.35), Rotation Vector=<0 0 1>)
   &SkeletalMeshLocation(
    Bone Sampling Mode=::NewEnumerator4
    MaskSpawnGroup=false
    Orientation Sampling=::NewEnumerator1
    Sampled Position Offset=<-2.5 0 -2.5>
    Skeletal Mesh=$System.Skeletal Mesh
   )
   &ScaleMeshSize(Scale Factor=&VectorFromFloat(Value=&FloatFromCurve(FloatCurve=DataInterface() {propertyValues {Curve {keys [
          {arriveTangent=2.2587, interpMode=RCIM_Cubic, leaveTangent=2.2587, tangentMode=RCTM_User, time=0, value=0}
          {arriveTangent=0.0224, interpMode=RCIM_Cubic, leaveTangent=0.0224, tangentMode=RCTM_User, time=0.7, value=1}
          {arriveTangent=-9.7992, interpMode=RCIM_Cubic, leaveTangent=-9.7992, tangentMode=RCTM_User, time=1, value=0.25}
         ]}}})))
   &ScaleColor(
    Scale Alpha=&FloatFromCurve(FloatCurve=DataInterface() {propertyValues {Curve {keys [
         {interpMode=RCIM_Cubic, tangentMode=RCTM_User, time=0, value=0}
         {arriveTangent=2.9731, interpMode=RCIM_Cubic, leaveTangent=1.0416, tangentMode=RCTM_Break, time=0.8, value=1}
         {arriveTangent=-14.195, interpMode=RCIM_Cubic, leaveTangent=-14.195, tangentMode=RCTM_User, time=1, value=0}
        ]}}})
    ScaleA=true
    ScaleRGB=true
   )
  ]
  rendererClasses [&NiagaraMeshRendererProperties]
  renderers [{rendererClass=&NiagaraMeshRendererProperties, rendererIndex=0}]
 }
}
summary {
 systemName=NS_Emote_VectorSpark_Shadow_V2
 emitters [
  {emitterName=SubUV_ConeGlowSkeletal, rendererClasses [&NiagaraSpriteRendererProperties]}
  {emitterName=LightCone_FromHand, rendererClasses [&NiagaraMeshRendererProperties]}
  {emitterName=Wave001, rendererClasses [&NiagaraMeshRendererProperties]}
  {emitterName=Mask001, rendererClasses [&NiagaraMeshRendererProperties]}
 ]
 userVariables [User.bIsFrontendPreview {type=NiagaraBool}]
}
